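; ChildProcMesh is a Blueprint in the Landmass old landscape prototype that declares an interface for making scatter meshes.
; MakeScatterMesh takes a BP Landmass Layer Stack reference, while MakeCustomScatterMesh takes no parameters.
; Both functions, the construction script and the BeginPlay, Tick and actor overlap events show no logic, so the file acts as a stub.
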(event EventBeginPlay)

(event Collision|EventActorBeginOverlap
  (params (OtherActor :type Actor%20Object%20Reference)))

(event EventTick
  (params (DeltaSeconds :type Float%20%28single-precision%29)))

(fn MakeCustomScatterMesh)

(fn MakeScatterMesh
  (params (LS%20BP :type BP%20Landmass%20Layer%20Stack%20Object%20Reference)))

(fn ConstructionScript)
