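; Prototype erosion component for BP_Landmass_LayerStack that runs GPU landscape erosion through material draws into ping-ponged height, velocity and position render targets.
; Erosion Tick and Button Press loop capped iterations of the selected mode: free particle scatter, shallow water advection, scatter-gather or water levelling.
; Particle scatter captures an instanced mesh via scene capture to erode height, and can periodically reseed particles.
; Each iteration also draws erosion, deposition and velocity effects into chosen weightmap channels.
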(event Custom|Erosion%20Tick
  (bind _erosion_data (Variables|RenderTargets|GetErosion%20Data))
  (bind (_fluidvela_10_9f4ddaaf414cc824f3d44a874df609ae _fluidvelb_13_30c22d1d49b68e204abf5a8372ae55dd _posrta_28_ebfe21624a8107e8a8379a90ea9a1e5a _posrtb_27_52922dee4c61f2eff44b7684b95e2e68) (Utilities|Struct|BreakErosionInterface _erosion_data))
  (bind _owninglayerstackbp (Variables|Default|GetOwningLayerStackBP))
  (bind _reset_every_n_interations (Variables|/Landmass/Landscape/OldPrototype_BP/BP_Landmass_LayerStack.BP_Landmass_LayerStack_C|GetReset%20Every%20N%20Interations _owninglayerstackbp))
  (bind _particle_position_seed_mid (Variables|MIDs|GetParticle%20Position%20Seed%20MID))
  (bind (_source_rt _target_rt) (Utilities|VelocityPingPong))
  (bind (_source_rt_2 _target_rt_2) (Utilities|PositionPingPong))
  (bind (_source_rt_3 _target_rt_3) (Utilities|HeightPingPong))
  (bind (_source_rt_4 _target_rt_4) (Utilities|VelocityPingPong))
  (bind _owninglayerstackbp_2 (Variables|Default|GetOwningLayerStackBP))
  (bind _scatter_particles_ismc (Variables|/Landmass/Landscape/OldPrototype_BP/BP_Landmass_LayerStack.BP_Landmass_LayerStack_C|GetScatter%20Particles%20ISMC _owninglayerstackbp_2))
  (bind _scatter_position_mid (Variables|MIDs|GetScatter%20Position%20MID))
  (bind _scatter_particle_render_mid (Variables|MIDs|GetScatter%20Particle%20Render%20MID))
  (bind _scatter_velocity_mid (Variables|MIDs|GetScatter%20Velocity%20MID))
  (bind (_source_rt_5 _target_rt_5) (Utilities|HeightPingPong))
  (bind (_source_rt_6 _target_rt_6) (Utilities|VelocityPingPong))
  (bind (_source_rt_7 _target_rt_7) (Utilities|VelocityPingPong))
  (bind (_source_rt_8 _target_rt_8) (Utilities|HeightPingPong))
  (bind _divergence_height_update_mid (Variables|MIDs|GetDivergence%20Height%20Update%20MID))
  (bind _advect_velocity_mid (Variables|MIDs|GetAdvect%20Velocity%20MID))
  (bind _advect_height_mid (Variables|MIDs|GetAdvect%20Height%20MID))
  (bind _velocity_acceleration_mid (Variables|MIDs|GetVelocity%20Acceleration%20MID))
  (bind (_source_rt_9 _target_rt_9) (Utilities|VelocityPingPong))
  (bind (_source_rt_10 _target_rt_10) (Utilities|HeightPingPong))
  (bind (_source_rt_11 _target_rt_11) (Utilities|VelocityPingPong))
  (bind (_source_rt_12 _target_rt_12) (Utilities|VelocityPingPong))
  (bind _velocity_acceleration_mid_2 (Variables|MIDs|GetVelocity%20Acceleration%20MID))
  (bind _advect_velocity_mid_2 (Variables|MIDs|GetAdvect%20Velocity%20MID))
  (bind _advect_height_mid_2 (Variables|MIDs|GetAdvect%20Height%20MID))
  (bind (_source_rt_13 _target_rt_13) (Utilities|HeightPingPong))
  (for _ (range (+ (Math|Integer|Min(Integer) (- (Variables|/Landmass/Landscape/OldPrototype_BP/BP_Landmass_LayerStack.BP_Landmass_LayerStack_C|GetPer%20Tick%20Iterations (Variables|Default|GetOwningLayerStackBP)) 1) 32) 1))
    (Utilities|FlowControl|Sequence
      (:then_0
        (if (Variables|/Landmass/Landscape/OldPrototype_BP/BP_Landmass_LayerStack.BP_Landmass_LayerStack_C|GetPerf%20Test%20Mode (Variables|Default|GetOwningLayerStackBP))
          (Rendering|DrawMaterialtoRenderTarget 0 "/Landmass/Landscape/OldPrototype_BP/Textures/RT_1x1.RT_1x1" "/Landmass/Landscape/OldPrototype_BP/Mats/Seed/UnpackRG8Height.UnpackRG8Height")))
      (:then_1
        (switch Utilities|FlowControl|Switch|SwitchonErosionMode (Variables|/Landmass/Landscape/OldPrototype_BP/BP_Landmass_LayerStack.BP_Landmass_LayerStack_C|GetErosion%20Mode (Variables|Default|GetOwningLayerStackBP))
          (:NewEnumerator0
            ; Free Particles Scatter
            (Utilities|FlowControl|Sequence
              (:then_0
                (Rendering|Material|SetTextureParameterValue _scatter_velocity_mid "HeightRT" (Utilities|HeightPingPong))
                (Rendering|Material|SetTextureParameterValue _scatter_velocity_mid "PositionRT" (Utilities|PositionPingPong))
                (Rendering|Material|SetTextureParameterValue _scatter_velocity_mid "VelocityRT" _source_rt)
                (Rendering|ClearRenderTarget2D 0 _target_rt)
                (Rendering|DrawMaterialtoRenderTarget 0 _target_rt (Variables|MIDs|GetScatter%20Velocity%20MID))
                (Math|Integer|IncrementInt (Variables|Debug|GetVelocityIdx)))
              (:then_2
                (Rendering|Material|SetTextureParameterValue _scatter_position_mid "HeightRT" (Utilities|HeightPingPong))
                (Rendering|Material|SetTextureParameterValue _scatter_position_mid "VelocityRT" (Utilities|VelocityPingPong))
                (Rendering|Material|SetTextureParameterValue _scatter_position_mid "PositionRT" _source_rt_2)
                (Rendering|ClearRenderTarget2D 0 _target_rt_2)
                (Rendering|DrawMaterialtoRenderTarget 0 _target_rt_2 (Variables|MIDs|GetScatter%20Position%20MID))
                (Math|Integer|IncrementInt (Variables|Debug|GetPositionIdx)))
              (:then_4
                (Rendering|Material|SetMaterial (Variables|/Landmass/Landscape/OldPrototype_BP/BP_Landmass_LayerStack.BP_Landmass_LayerStack_C|GetScatter%20Particles%20ISMC (Variables|Default|GetOwningLayerStackBP)) 0 (Variables|MIDs|GetScatter%20Particle%20Render%20MID))
                (Rendering|Material|SetTextureParameterValue _scatter_particle_render_mid "VelocityRT" _source_rt_4)
                (Rendering|Material|SetTextureParameterValue _scatter_particle_render_mid "VelocityRTPrev" _target_rt_4)
                (Rendering|Material|SetTextureParameterValue (Variables|MIDs|GetScatter%20Particle%20Preview%20MID) "PositionRT" (Utilities|PositionPingPong))
                (Rendering|SetVisibility _scatter_particles_ismc true false)
                (Rendering|SceneCapture|CaptureScene (Variables|/Landmass/Landscape/OldPrototype_BP/BP_Landmass_LayerStack.BP_Landmass_LayerStack_C|GetScatterMeshSceneCapture _owninglayerstackbp_2) :!class /Script/Engine.SceneCaptureComponent2D)
                (Rendering|Material|SetMaterial (Variables|/Landmass/Landscape/OldPrototype_BP/BP_Landmass_LayerStack.BP_Landmass_LayerStack_C|GetScatter%20Particles%20ISMC (Variables|Default|GetOwningLayerStackBP)) 0 (Variables|MIDs|GetScatter%20Particle%20Preview%20MID))
                (Rendering|SetVisibility _scatter_particles_ismc (Variables|/Landmass/Landscape/OldPrototype_BP/BP_Landmass_LayerStack.BP_Landmass_LayerStack_C|GetShow%20Water (Variables|Default|GetOwningLayerStackBP)) false))
              (:then_6
                (Rendering|Material|SetTextureParameterValue (Variables|MIDs|GetScatter%20Height%20Erosion%20MID) "HeightRT" _source_rt_3)
                (Rendering|DrawMaterialtoRenderTarget 0 _target_rt_3 (Variables|MIDs|GetScatter%20Height%20Erosion%20MID))
                (Math|Integer|IncrementInt (Variables|/Landmass/Landscape/OldPrototype_BP/BP_Landmass_LayerStack.BP_Landmass_LayerStack_C|GetStackHeightIdx (Variables|Default|GetOwningLayerStackBP))
                  (:%20%20
                    ;; Reset Iterations
                    (if (> _reset_every_n_interations 0)
                      (if (== (Math|Integer|%25%28Integer%29 (Variables|/Landmass/Landscape/OldPrototype_BP/BP_Landmass_LayerStack.BP_Landmass_LayerStack_C|GetStackHeightIdx _owninglayerstackbp) _reset_every_n_interations) 0)
                        (Rendering|ClearRenderTarget2D 0 _fluidvela_10_9f4ddaaf414cc824f3d44a874df609ae "(R=0.000000,G=0.000000,B=0.000000,A=0.000000)")
                        (Rendering|ClearRenderTarget2D 0 _fluidvelb_13_30c22d1d49b68e204abf5a8372ae55dd "(R=0.000000,G=0.000000,B=0.000000,A=0.000000)")
                        (Rendering|ClearRenderTarget2D 0 _posrta_28_ebfe21624a8107e8a8379a90ea9a1e5a)
                        (Rendering|ClearRenderTarget2D 0 _posrtb_27_52922dee4c61f2eff44b7684b95e2e68)
                        (Rendering|Material|SetTextureParameterValue (Variables|MIDs|GetParticle%20Position%20Seed%20MID) "HeightRT" (Utilities|HeightPingPong))
                        (Rendering|DrawMaterialtoRenderTarget 0 _posrta_28_ebfe21624a8107e8a8379a90ea9a1e5a _particle_position_seed_mid)
                        (Rendering|DrawMaterialtoRenderTarget 0 _posrtb_27_52922dee4c61f2eff44b7684b95e2e68 _particle_position_seed_mid)
                        (Variables|/Landmass/Landscape/OldPrototype_BP/BP_Landmass_LayerStack.BP_Landmass_LayerStack_C|SetStackHeightIdx (Variables|/Landmass/Landscape/OldPrototype_BP/BP_Landmass_LayerStack.BP_Landmass_LayerStack_C|GetReset%20Every%20N%20Interations (Variables|Default|GetOwningLayerStackBP)) (Variables|Default|GetOwningLayerStackBP))
                        (Variables|Debug|SetVelocityIdx 0)
                        (Variables|Debug|SetPositionIdx 0)
                        (Variables|Debug|SetIterationIdx 0))))))))
          (:NewEnumerator1
            ; Shallow Water Advection
            (Utilities|FlowControl|Sequence
              (:then_0
                (Rendering|Material|SetScalarParameterValue _advect_height_mid "idx" (Math|Conversions|ToFloat(Integer) (Variables|Debug|GetIterationIdx)) :!class /Script/Engine.MaterialInstanceDynamic)
                (Rendering|Material|SetTextureParameterValue _advect_height_mid "VelocityRT" (Utilities|VelocityPingPong))
                (Rendering|Material|SetTextureParameterValue _advect_height_mid "HeightRT" (Utilities|HeightPingPong))
                (Rendering|ClearRenderTarget2D 0 _target_rt_8)
                (Rendering|DrawMaterialtoRenderTarget 0 _target_rt_8 (Variables|MIDs|GetAdvect%20Height%20MID))
                (Math|Integer|IncrementInt (Variables|/Landmass/Landscape/OldPrototype_BP/BP_Landmass_LayerStack.BP_Landmass_LayerStack_C|GetStackHeightIdx (Variables|Default|GetOwningLayerStackBP))))
              (:then_2
                (Rendering|Material|SetScalarParameterValue _advect_velocity_mid "idx" (Math|Conversions|ToFloat(Integer) (Variables|Debug|GetIterationIdx)) :!class /Script/Engine.MaterialInstanceDynamic)
                (Rendering|Material|SetTextureParameterValue _advect_velocity_mid "VelocityRT" _source_rt_7)
                (Rendering|ClearRenderTarget2D 0 _target_rt_7)
                (Rendering|DrawMaterialtoRenderTarget 0 _target_rt_7 (Variables|MIDs|GetAdvect%20Velocity%20MID))
                (Math|Integer|IncrementInt (Variables|Debug|GetVelocityIdx)))
              (:then_4
                (Rendering|Material|SetTextureParameterValue _divergence_height_update_mid "VelocityRT" (Utilities|VelocityPingPong)) ; Height contains water info
                (Rendering|Material|SetTextureParameterValue _divergence_height_update_mid "HeightRT" _source_rt_5)
                (Rendering|ClearRenderTarget2D 0 _target_rt_5)
                (Rendering|DrawMaterialtoRenderTarget 0 _target_rt_5 (Variables|MIDs|GetDivergence%20Height%20Update%20MID))
                (Math|Integer|IncrementInt (Variables|/Landmass/Landscape/OldPrototype_BP/BP_Landmass_LayerStack.BP_Landmass_LayerStack_C|GetStackHeightIdx (Variables|Default|GetOwningLayerStackBP))))
              (:then_6
                (Rendering|Material|SetTextureParameterValue _velocity_acceleration_mid "HeightRT" (Utilities|HeightPingPong)) ; Height contains water info
                (Rendering|Material|SetTextureParameterValue _velocity_acceleration_mid "VelocityRT" _source_rt_6)
                (Rendering|ClearRenderTarget2D 0 _target_rt_6)
                (Rendering|DrawMaterialtoRenderTarget 0 _target_rt_6 (Variables|MIDs|GetVelocity%20Acceleration%20MID))
                (Math|Integer|IncrementInt (Variables|Debug|GetVelocityIdx)))))
          (:NewEnumerator2
            ; Scatter Gather Method
            (Rendering|Material|SetTextureParameterValue _velocity_acceleration_mid_2 "HeightRT" (Utilities|HeightPingPong)) ; Height contains water info
            (Rendering|Material|SetTextureParameterValue _velocity_acceleration_mid_2 "VelocityRT" _source_rt_9)
            (Rendering|ClearRenderTarget2D 0 _target_rt_9)
            (Rendering|DrawMaterialtoRenderTarget 0 _target_rt_9 (Variables|MIDs|GetVelocity%20Acceleration%20MID))
            (Math|Integer|IncrementInt (Variables|Debug|GetVelocityIdx)
              (:%20%20
                (Utilities|FlowControl|Sequence
                  (:then_0
                    (Rendering|Material|SetTextureParameterValue _advect_velocity_mid_2 "HeightRT" (Utilities|HeightPingPong))
                    (Rendering|Material|SetTextureParameterValue _advect_velocity_mid_2 "VelocityRT" _source_rt_11)
                    (Rendering|ClearRenderTarget2D 0 _target_rt_11)
                    (Rendering|DrawMaterialtoRenderTarget 0 _target_rt_11 (Variables|MIDs|GetAdvect%20Velocity%20MID))
                    (Math|Integer|IncrementInt (Variables|Debug|GetVelocityIdx)))
                  (:then_2
                    (Rendering|Material|SetTextureParameterValue _advect_height_mid_2 "HeightRT" (Utilities|HeightPingPong))
                    (Rendering|Material|SetTextureParameterValue _advect_height_mid_2 "VelocityRTCurrent" _source_rt_12)
                    (Rendering|Material|SetTextureParameterValue _advect_height_mid_2 "VelocityRT" _target_rt_12)
                    (Rendering|ClearRenderTarget2D 0 _target_rt_10)
                    (Rendering|DrawMaterialtoRenderTarget 0 _target_rt_10 (Variables|MIDs|GetAdvect%20Height%20MID))
                    (Math|Integer|IncrementInt (Variables|/Landmass/Landscape/OldPrototype_BP/BP_Landmass_LayerStack.BP_Landmass_LayerStack_C|GetStackHeightIdx (Variables|Default|GetOwningLayerStackBP))))))))
          (:NewEnumerator3
            ; Simple Water Levelling
            (Rendering|Material|SetTextureParameterValue (Variables|MIDs|GetAdvect%20Height%20MID) "HeightRT" _source_rt_13)
            (Rendering|ClearRenderTarget2D 0 _target_rt_13)
            (Rendering|DrawMaterialtoRenderTarget 0 _target_rt_13 (Variables|MIDs|GetAdvect%20Height%20MID))
            (Math|Integer|IncrementInt (Variables|/Landmass/Landscape/OldPrototype_BP/BP_Landmass_LayerStack.BP_Landmass_LayerStack_C|GetStackHeightIdx (Variables|Default|GetOwningLayerStackBP))))))
      (:then_2
        (Math|Integer|IncrementInt (Variables|Debug|GetIterationIdx)
          (:%20%20
            (Rendering|Material|SetScalarParameterValue 0 "/Landmass/Landscape/BlueprintBrushes/MPC/MPC_Erosion.MPC_Erosion" "idx" (Math|Conversions|ToFloat(Integer) (Variables|Debug|GetIterationIdx)) :!class /Script/Engine.KismetMaterialLibrary)
            (CallFunction|RenderWeightmapEffects))))))
  (Class|BPLandmassLayerStack|UpdateNormalandPreviewRTAssignments (Variables|Default|GetOwningLayerStackBP) (Utilities|VelocityPingPong))
  (if (Utilities|Enum|Equal(Enum) (Variables|/Landmass/Landscape/OldPrototype_BP/BP_Landmass_LayerStack.BP_Landmass_LayerStack_C|GetErosion%20Mode (Variables|Default|GetOwningLayerStackBP)) "NewEnumerator0")
    _))

(event Custom|Erosion%20Button%20Press
  (bind _erosion_data (Variables|RenderTargets|GetErosion%20Data))
  (bind (_fluidvela_10_9f4ddaaf414cc824f3d44a874df609ae _fluidvelb_13_30c22d1d49b68e204abf5a8372ae55dd _posrta_28_ebfe21624a8107e8a8379a90ea9a1e5a _posrtb_27_52922dee4c61f2eff44b7684b95e2e68) (Utilities|Struct|BreakErosionInterface _erosion_data))
  (bind _owninglayerstackbp (Variables|Default|GetOwningLayerStackBP))
  (bind _reset_every_n_interations (Variables|/Landmass/Landscape/OldPrototype_BP/BP_Landmass_LayerStack.BP_Landmass_LayerStack_C|GetReset%20Every%20N%20Interations _owninglayerstackbp))
  (bind _particle_position_seed_mid (Variables|MIDs|GetParticle%20Position%20Seed%20MID))
  (bind (_source_rt _target_rt) (Utilities|VelocityPingPong))
  (bind (_source_rt_2 _target_rt_2) (Utilities|PositionPingPong))
  (bind (_source_rt_3 _target_rt_3) (Utilities|HeightPingPong))
  (bind (_source_rt_4 _target_rt_4) (Utilities|VelocityPingPong))
  (bind _owninglayerstackbp_2 (Variables|Default|GetOwningLayerStackBP))
  (bind _scatter_particles_ismc (Variables|/Landmass/Landscape/OldPrototype_BP/BP_Landmass_LayerStack.BP_Landmass_LayerStack_C|GetScatter%20Particles%20ISMC _owninglayerstackbp_2))
  (bind _scatter_position_mid (Variables|MIDs|GetScatter%20Position%20MID))
  (bind _scatter_particle_render_mid (Variables|MIDs|GetScatter%20Particle%20Render%20MID))
  (bind _scatter_velocity_mid (Variables|MIDs|GetScatter%20Velocity%20MID))
  (bind (_source_rt_5 _target_rt_5) (Utilities|HeightPingPong))
  (bind (_source_rt_6 _target_rt_6) (Utilities|VelocityPingPong))
  (bind (_source_rt_7 _target_rt_7) (Utilities|VelocityPingPong))
  (bind (_source_rt_8 _target_rt_8) (Utilities|HeightPingPong))
  (bind _divergence_height_update_mid (Variables|MIDs|GetDivergence%20Height%20Update%20MID))
  (bind _advect_velocity_mid (Variables|MIDs|GetAdvect%20Velocity%20MID))
  (bind _advect_height_mid (Variables|MIDs|GetAdvect%20Height%20MID))
  (bind _velocity_acceleration_mid (Variables|MIDs|GetVelocity%20Acceleration%20MID))
  (bind (_source_rt_9 _target_rt_9) (Utilities|VelocityPingPong))
  (bind (_source_rt_10 _target_rt_10) (Utilities|HeightPingPong))
  (bind (_source_rt_11 _target_rt_11) (Utilities|VelocityPingPong))
  (bind (_source_rt_12 _target_rt_12) (Utilities|VelocityPingPong))
  (bind _velocity_acceleration_mid_2 (Variables|MIDs|GetVelocity%20Acceleration%20MID))
  (bind _advect_velocity_mid_2 (Variables|MIDs|GetAdvect%20Velocity%20MID))
  (bind _advect_height_mid_2 (Variables|MIDs|GetAdvect%20Height%20MID))
  (bind (_source_rt_13 _target_rt_13) (Utilities|HeightPingPong))
  (for _ (range (+ (Math|Integer|Min(Integer) (- (Variables|/Landmass/Landscape/OldPrototype_BP/BP_Landmass_LayerStack.BP_Landmass_LayerStack_C|GetButton%20Press%20Iterations (Variables|Default|GetOwningLayerStackBP)) 1) 32) 1))
    (Utilities|FlowControl|Sequence
      (:then_0
        (if (Variables|/Landmass/Landscape/OldPrototype_BP/BP_Landmass_LayerStack.BP_Landmass_LayerStack_C|GetPerf%20Test%20Mode (Variables|Default|GetOwningLayerStackBP))
          (Rendering|DrawMaterialtoRenderTarget 0 "/Landmass/Landscape/OldPrototype_BP/Textures/RT_1x1.RT_1x1" "/Landmass/Landscape/OldPrototype_BP/Mats/Seed/UnpackRG8Height.UnpackRG8Height")))
      (:then_1
        (switch Utilities|FlowControl|Switch|SwitchonErosionMode (Variables|/Landmass/Landscape/OldPrototype_BP/BP_Landmass_LayerStack.BP_Landmass_LayerStack_C|GetErosion%20Mode (Variables|Default|GetOwningLayerStackBP))
          (:NewEnumerator0
            ; Free Particles Scatter
            (Utilities|FlowControl|Sequence
              (:then_0
                (Rendering|Material|SetTextureParameterValue _scatter_velocity_mid "HeightRT" (Utilities|HeightPingPong))
                (Rendering|Material|SetTextureParameterValue _scatter_velocity_mid "PositionRT" (Utilities|PositionPingPong))
                (Rendering|Material|SetTextureParameterValue _scatter_velocity_mid "VelocityRT" _source_rt)
                (Rendering|ClearRenderTarget2D 0 _target_rt)
                (Rendering|DrawMaterialtoRenderTarget 0 _target_rt (Variables|MIDs|GetScatter%20Velocity%20MID))
                (Math|Integer|IncrementInt (Variables|Debug|GetVelocityIdx)))
              (:then_2
                (Rendering|Material|SetTextureParameterValue _scatter_position_mid "HeightRT" (Utilities|HeightPingPong))
                (Rendering|Material|SetTextureParameterValue _scatter_position_mid "VelocityRT" (Utilities|VelocityPingPong))
                (Rendering|Material|SetTextureParameterValue _scatter_position_mid "PositionRT" _source_rt_2)
                (Rendering|ClearRenderTarget2D 0 _target_rt_2)
                (Rendering|DrawMaterialtoRenderTarget 0 _target_rt_2 (Variables|MIDs|GetScatter%20Position%20MID))
                (Math|Integer|IncrementInt (Variables|Debug|GetPositionIdx)))
              (:then_4
                (Rendering|Material|SetMaterial (Variables|/Landmass/Landscape/OldPrototype_BP/BP_Landmass_LayerStack.BP_Landmass_LayerStack_C|GetScatter%20Particles%20ISMC (Variables|Default|GetOwningLayerStackBP)) 0 (Variables|MIDs|GetScatter%20Particle%20Render%20MID))
                (Rendering|Material|SetTextureParameterValue _scatter_particle_render_mid "VelocityRT" _source_rt_4)
                (Rendering|Material|SetTextureParameterValue _scatter_particle_render_mid "VelocityRTPrev" _target_rt_4)
                (Rendering|Material|SetTextureParameterValue (Variables|MIDs|GetScatter%20Particle%20Preview%20MID) "PositionRT" (Utilities|PositionPingPong))
                (Rendering|SetVisibility _scatter_particles_ismc true false)
                (Rendering|SceneCapture|CaptureScene (Variables|/Landmass/Landscape/OldPrototype_BP/BP_Landmass_LayerStack.BP_Landmass_LayerStack_C|GetScatterMeshSceneCapture _owninglayerstackbp_2) :!class /Script/Engine.SceneCaptureComponent2D)
                (Rendering|Material|SetMaterial (Variables|/Landmass/Landscape/OldPrototype_BP/BP_Landmass_LayerStack.BP_Landmass_LayerStack_C|GetScatter%20Particles%20ISMC (Variables|Default|GetOwningLayerStackBP)) 0 (Variables|MIDs|GetScatter%20Particle%20Preview%20MID))
                (Rendering|SetVisibility _scatter_particles_ismc (Variables|/Landmass/Landscape/OldPrototype_BP/BP_Landmass_LayerStack.BP_Landmass_LayerStack_C|GetShow%20Water (Variables|Default|GetOwningLayerStackBP)) false))
              (:then_6
                (Rendering|Material|SetTextureParameterValue (Variables|MIDs|GetScatter%20Height%20Erosion%20MID) "HeightRT" _source_rt_3)
                (Rendering|DrawMaterialtoRenderTarget 0 _target_rt_3 (Variables|MIDs|GetScatter%20Height%20Erosion%20MID))
                (Math|Integer|IncrementInt (Variables|/Landmass/Landscape/OldPrototype_BP/BP_Landmass_LayerStack.BP_Landmass_LayerStack_C|GetStackHeightIdx (Variables|Default|GetOwningLayerStackBP))
                  (:%20%20
                    ;; Reset Iterations
                    (if (> _reset_every_n_interations 0)
                      (if (== (Math|Integer|%25%28Integer%29 (Variables|/Landmass/Landscape/OldPrototype_BP/BP_Landmass_LayerStack.BP_Landmass_LayerStack_C|GetStackHeightIdx _owninglayerstackbp) _reset_every_n_interations) 0)
                        (Rendering|ClearRenderTarget2D 0 _fluidvela_10_9f4ddaaf414cc824f3d44a874df609ae "(R=0.000000,G=0.000000,B=0.000000,A=0.000000)")
                        (Rendering|ClearRenderTarget2D 0 _fluidvelb_13_30c22d1d49b68e204abf5a8372ae55dd "(R=0.000000,G=0.000000,B=0.000000,A=0.000000)")
                        (Rendering|ClearRenderTarget2D 0 _posrta_28_ebfe21624a8107e8a8379a90ea9a1e5a)
                        (Rendering|ClearRenderTarget2D 0 _posrtb_27_52922dee4c61f2eff44b7684b95e2e68)
                        (Rendering|Material|SetTextureParameterValue (Variables|MIDs|GetParticle%20Position%20Seed%20MID) "HeightRT" (Utilities|HeightPingPong))
                        (Rendering|DrawMaterialtoRenderTarget 0 _posrta_28_ebfe21624a8107e8a8379a90ea9a1e5a _particle_position_seed_mid)
                        (Rendering|DrawMaterialtoRenderTarget 0 _posrtb_27_52922dee4c61f2eff44b7684b95e2e68 _particle_position_seed_mid)
                        (Variables|/Landmass/Landscape/OldPrototype_BP/BP_Landmass_LayerStack.BP_Landmass_LayerStack_C|SetStackHeightIdx (Variables|/Landmass/Landscape/OldPrototype_BP/BP_Landmass_LayerStack.BP_Landmass_LayerStack_C|GetReset%20Every%20N%20Interations (Variables|Default|GetOwningLayerStackBP)) (Variables|Default|GetOwningLayerStackBP))
                        (Variables|Debug|SetVelocityIdx 0)
                        (Variables|Debug|SetPositionIdx 0)
                        (Variables|Debug|SetIterationIdx 0))))))))
          (:NewEnumerator1
            ; Shallow Water Advection
            (Utilities|FlowControl|Sequence
              (:then_0
                (Rendering|Material|SetScalarParameterValue _advect_height_mid "idx" (Math|Conversions|ToFloat(Integer) (Variables|Debug|GetIterationIdx)) :!class /Script/Engine.MaterialInstanceDynamic)
                (Rendering|Material|SetTextureParameterValue _advect_height_mid "VelocityRT" (Utilities|VelocityPingPong))
                (Rendering|Material|SetTextureParameterValue _advect_height_mid "HeightRT" (Utilities|HeightPingPong))
                (Rendering|ClearRenderTarget2D 0 _target_rt_8)
                (Rendering|DrawMaterialtoRenderTarget 0 _target_rt_8 (Variables|MIDs|GetAdvect%20Height%20MID))
                (Math|Integer|IncrementInt (Variables|/Landmass/Landscape/OldPrototype_BP/BP_Landmass_LayerStack.BP_Landmass_LayerStack_C|GetStackHeightIdx (Variables|Default|GetOwningLayerStackBP))))
              (:then_2
                (Rendering|Material|SetScalarParameterValue _advect_velocity_mid "idx" (Math|Conversions|ToFloat(Integer) (Variables|Debug|GetIterationIdx)) :!class /Script/Engine.MaterialInstanceDynamic)
                (Rendering|Material|SetTextureParameterValue _advect_velocity_mid "VelocityRT" _source_rt_7)
                (Rendering|ClearRenderTarget2D 0 _target_rt_7)
                (Rendering|DrawMaterialtoRenderTarget 0 _target_rt_7 (Variables|MIDs|GetAdvect%20Velocity%20MID))
                (Math|Integer|IncrementInt (Variables|Debug|GetVelocityIdx)))
              (:then_4
                (Rendering|Material|SetTextureParameterValue _divergence_height_update_mid "VelocityRT" (Utilities|VelocityPingPong)) ; Height contains water info
                (Rendering|Material|SetTextureParameterValue _divergence_height_update_mid "HeightRT" _source_rt_5)
                (Rendering|ClearRenderTarget2D 0 _target_rt_5)
                (Rendering|DrawMaterialtoRenderTarget 0 _target_rt_5 (Variables|MIDs|GetDivergence%20Height%20Update%20MID))
                (Math|Integer|IncrementInt (Variables|/Landmass/Landscape/OldPrototype_BP/BP_Landmass_LayerStack.BP_Landmass_LayerStack_C|GetStackHeightIdx (Variables|Default|GetOwningLayerStackBP))))
              (:then_6
                (Rendering|Material|SetTextureParameterValue _velocity_acceleration_mid "HeightRT" (Utilities|HeightPingPong)) ; Height contains water info
                (Rendering|Material|SetTextureParameterValue _velocity_acceleration_mid "VelocityRT" _source_rt_6)
                (Rendering|ClearRenderTarget2D 0 _target_rt_6)
                (Rendering|DrawMaterialtoRenderTarget 0 _target_rt_6 (Variables|MIDs|GetVelocity%20Acceleration%20MID))
                (Math|Integer|IncrementInt (Variables|Debug|GetVelocityIdx)))))
          (:NewEnumerator2
            ; Scatter Gather Method
            (Rendering|Material|SetTextureParameterValue _velocity_acceleration_mid_2 "HeightRT" (Utilities|HeightPingPong)) ; Height contains water info
            (Rendering|Material|SetTextureParameterValue _velocity_acceleration_mid_2 "VelocityRT" _source_rt_9)
            (Rendering|ClearRenderTarget2D 0 _target_rt_9)
            (Rendering|DrawMaterialtoRenderTarget 0 _target_rt_9 (Variables|MIDs|GetVelocity%20Acceleration%20MID))
            (Math|Integer|IncrementInt (Variables|Debug|GetVelocityIdx)
              (:%20%20
                (Utilities|FlowControl|Sequence
                  (:then_0
                    (Rendering|Material|SetTextureParameterValue _advect_velocity_mid_2 "HeightRT" (Utilities|HeightPingPong))
                    (Rendering|Material|SetTextureParameterValue _advect_velocity_mid_2 "VelocityRT" _source_rt_11)
                    (Rendering|ClearRenderTarget2D 0 _target_rt_11)
                    (Rendering|DrawMaterialtoRenderTarget 0 _target_rt_11 (Variables|MIDs|GetAdvect%20Velocity%20MID))
                    (Math|Integer|IncrementInt (Variables|Debug|GetVelocityIdx)))
                  (:then_2
                    (Rendering|Material|SetTextureParameterValue _advect_height_mid_2 "HeightRT" (Utilities|HeightPingPong))
                    (Rendering|Material|SetTextureParameterValue _advect_height_mid_2 "VelocityRTCurrent" _source_rt_12)
                    (Rendering|Material|SetTextureParameterValue _advect_height_mid_2 "VelocityRT" _target_rt_12)
                    (Rendering|ClearRenderTarget2D 0 _target_rt_10)
                    (Rendering|DrawMaterialtoRenderTarget 0 _target_rt_10 (Variables|MIDs|GetAdvect%20Height%20MID))
                    (Math|Integer|IncrementInt (Variables|/Landmass/Landscape/OldPrototype_BP/BP_Landmass_LayerStack.BP_Landmass_LayerStack_C|GetStackHeightIdx (Variables|Default|GetOwningLayerStackBP))))))))
          (:NewEnumerator3
            ; Simple Water Levelling
            (Rendering|Material|SetTextureParameterValue (Variables|MIDs|GetAdvect%20Height%20MID) "HeightRT" _source_rt_13)
            (Rendering|ClearRenderTarget2D 0 _target_rt_13)
            (Rendering|DrawMaterialtoRenderTarget 0 _target_rt_13 (Variables|MIDs|GetAdvect%20Height%20MID))
            (Math|Integer|IncrementInt (Variables|/Landmass/Landscape/OldPrototype_BP/BP_Landmass_LayerStack.BP_Landmass_LayerStack_C|GetStackHeightIdx (Variables|Default|GetOwningLayerStackBP))))))
      (:then_2
        (Math|Integer|IncrementInt (Variables|Debug|GetIterationIdx)
          (:%20%20
            (Rendering|Material|SetScalarParameterValue 0 "/Landmass/Landscape/BlueprintBrushes/MPC/MPC_Erosion.MPC_Erosion" "idx" (Math|Conversions|ToFloat(Integer) (Variables|Debug|GetIterationIdx)) :!class /Script/Engine.KismetMaterialLibrary)
            (CallFunction|RenderWeightmapEffects))))))
  (Class|BPLandmassLayerStack|UpdateNormalandPreviewRTAssignments (Variables|Default|GetOwningLayerStackBP) (Utilities|VelocityPingPong))
  (if (Utilities|Enum|Equal(Enum) (Variables|/Landmass/Landscape/OldPrototype_BP/BP_Landmass_LayerStack.BP_Landmass_LayerStack_C|GetErosion%20Mode (Variables|Default|GetOwningLayerStackBP)) "NewEnumerator0")
    _))

(event Custom|Seed%20Particle%20Positions
  (bind _erosion_data (Variables|RenderTargets|GetErosion%20Data))
  (bind (_posrta_28_ebfe21624a8107e8a8379a90ea9a1e5a _posrtb_27_52922dee4c61f2eff44b7684b95e2e68) (Utilities|Struct|BreakErosionInterface _erosion_data))
  (bind _erosion_data_2 (Variables|RenderTargets|GetErosion%20Data))
  (bind (_heightrta_2_023115b74de9087cdfc3a58fb4a37a7c _heightrtb_4_8915f4844f348332719555ab670bfe75 _fluidvela_10_9f4ddaaf414cc824f3d44a874df609ae _fluidvelb_13_30c22d1d49b68e204abf5a8372ae55dd _scenecapturert_15_0c62938b463a03607a141487dc9c1a4c _scenecapturecomp_20_d7f752ce453c81cee5bf9ba9c0be8a27 _posrta_28_ebfe21624a8107e8a8379a90ea9a1e5a_2) (Utilities|Struct|BreakErosionInterface _erosion_data_2))
  (bind _particle_position_seed_mid (Variables|MIDs|GetParticle%20Position%20Seed%20MID))
  (Rendering|ClearRenderTarget2D 0 _posrta_28_ebfe21624a8107e8a8379a90ea9a1e5a)
  (Rendering|ClearRenderTarget2D 0 _posrtb_27_52922dee4c61f2eff44b7684b95e2e68)
  (Rendering|Material|SetTextureParameterValue _particle_position_seed_mid "HeightRT" (Utilities|HeightPingPong))
  (Rendering|DrawMaterialtoRenderTarget 0 _posrta_28_ebfe21624a8107e8a8379a90ea9a1e5a_2 _particle_position_seed_mid)
  (Rendering|Material|SetTextureParameterValue (Variables|MIDs|GetScatter%20Particle%20Preview%20MID) "PositionRT" _posrta_28_ebfe21624a8107e8a8379a90ea9a1e5a_2))

(fn Constructor
  (params (ErosionData :type Erosion%20Interface%20Structure))
  (bind _owninglayerstackbp (Variables|Default|GetOwningLayerStackBP))
  (bind _scatter_particles_ismc (Variables|/Landmass/Landscape/OldPrototype_BP/BP_Landmass_LayerStack.BP_Landmass_LayerStack_C|GetScatter%20Particles%20ISMC _owninglayerstackbp))
  (Variables|RenderTargets|SetErosion%20Data ErosionData)
  (bind _asbp_landmass_layer_stack (Utilities|Casting|CastToBP_Landmass_LayerStack (Components|GetOwner)))
  (Variables|Default|SetOwningLayerStackBP _asbp_landmass_layer_stack)
  (Utilities|FlowControl|Sequence
    (:then_1
      (Utilities|CreateMIDs))
    (:then_2
      (Utilities|SetMIDParams))
    (:then_3
      (if (Utilities|Enum|Equal(Enum) (Variables|/Landmass/Landscape/OldPrototype_BP/BP_Landmass_LayerStack.BP_Landmass_LayerStack_C|GetErosion%20Mode (Variables|Default|GetOwningLayerStackBP)) "NewEnumerator0")
        (Utilities|IsValid _scatter_particles_ismc
          (:Is%20Valid
            (Rendering|Material|SetMaterial _scatter_particles_ismc 0 (Variables|MIDs|GetScatter%20Particle%20Preview%20MID))))))))

(fn RenderWeightmapEffects
  (bind (_source_rt _target_rt) (Utilities|HeightPingPong))
  (bind _owninglayerstackbp (Variables|Default|GetOwningLayerStackBP))
  (bind _weightmapeffectsmid (Variables|MIDs|GetWeightmapEffectsMID))
  (bind _weightmaps (Variables|/Landmass/Landscape/OldPrototype_BP/BP_Landmass_LayerStack.BP_Landmass_LayerStack_C|GetWeightmaps _owninglayerstackbp))
  (for _array_element (Variables|/Landmass/Landscape/OldPrototype_BP/BP_Landmass_LayerStack.BP_Landmass_LayerStack_C|GetErosion%20Weightmap%20Effects (Variables|Default|GetOwningLayerStackBP))
    (if (.ReturnValue (Utilities|Map|Find _weightmaps (Utilities|String|StringToName (Utilities|Array|Get(aref) (Utilities|String|ParseIntoArray (Utilities|Struct|BreakErodeWeightmap _array_element) ".")))))
      (Utilities|IsValid (Utilities|Struct|BreakRenderTargetChannels (Utilities|Map|Find _weightmaps (Utilities|String|StringToName (Utilities|Array|Get(aref) (Utilities|String|ParseIntoArray (Utilities|Struct|BreakErodeWeightmap _array_element) ".")))))
        (:Is%20Valid
          (if (Utilities|Name|Equal(Name) (Utilities|String|StringToName (Utilities|Array|Get(aref) (Utilities|String|ParseIntoArray (Utilities|Struct|BreakErodeWeightmap _array_element) ".") 1)) (.R_6_6FEDD63A4C392A58DB7793AFC063FE06 (Utilities|Struct|BreakRenderTargetChannels (Utilities|Map|Find _weightmaps (Utilities|String|StringToName (Utilities|Array|Get(aref) (Utilities|String|ParseIntoArray (Utilities|Struct|BreakErodeWeightmap _array_element) ".")))))))
            (LocalVariables|SetChannelMask (Utilities|Struct|MakeLinearColor 1.0))
            (elif (Utilities|Name|Equal(Name) (Utilities|String|StringToName (Utilities|Array|Get(aref) (Utilities|String|ParseIntoArray (Utilities|Struct|BreakErodeWeightmap _array_element) ".") 1)) (.G_8_70B7D3C141576DD58ACE3E83E88865BC (Utilities|Struct|BreakRenderTargetChannels (Utilities|Map|Find _weightmaps (Utilities|String|StringToName (Utilities|Array|Get(aref) (Utilities|String|ParseIntoArray (Utilities|Struct|BreakErodeWeightmap _array_element) ".")))))))
              (LocalVariables|SetChannelMask (Utilities|Struct|MakeLinearColor 0.0 1.0))
              (elif (Utilities|Name|Equal(Name) (Utilities|String|StringToName (Utilities|Array|Get(aref) (Utilities|String|ParseIntoArray (Utilities|Struct|BreakErodeWeightmap _array_element) ".") 1)) (.B_10_BB4928164B8A11CB3E82F8A3161F40F5 (Utilities|Struct|BreakRenderTargetChannels (Utilities|Map|Find _weightmaps (Utilities|String|StringToName (Utilities|Array|Get(aref) (Utilities|String|ParseIntoArray (Utilities|Struct|BreakErodeWeightmap _array_element) ".")))))))
                (LocalVariables|SetChannelMask (Utilities|Struct|MakeLinearColor 0.0 0.0 1.0))
                (elif (Utilities|Name|Equal(Name) (Utilities|String|StringToName (Utilities|Array|Get(aref) (Utilities|String|ParseIntoArray (Utilities|Struct|BreakErodeWeightmap _array_element) ".") 1)) (.A_23_8074622A4536EA2CF5EEFF9B0027115F (Utilities|Struct|BreakRenderTargetChannels (Utilities|Map|Find _weightmaps (Utilities|String|StringToName (Utilities|Array|Get(aref) (Utilities|String|ParseIntoArray (Utilities|Struct|BreakErodeWeightmap _array_element) ".")))))))
                  (LocalVariables|SetChannelMask (Utilities|Struct|MakeLinearColor 0.0 0.0 0.0 1.0))))))
          (Rendering|Material|SetVectorParameterValue _weightmapeffectsmid "ChannelMask" (LocalVariables|GetChannelMask) :!class /Script/Engine.MaterialInstanceDynamic)
          (Rendering|Material|SetScalarParameterValue _weightmapeffectsmid "Deposition" (.DepositionStrength_9_A0C69D6145F47973775021AB8749B742 (Utilities|Struct|BreakErodeWeightmap _array_element)) :!class /Script/Engine.MaterialInstanceDynamic)
          (Rendering|Material|SetScalarParameterValue _weightmapeffectsmid "Erosion" (.ErosionStrength_10_CF946B9A4ADA64BEEB4E8DA6A2CEA428 (Utilities|Struct|BreakErodeWeightmap _array_element)) :!class /Script/Engine.MaterialInstanceDynamic)
          (Rendering|Material|SetScalarParameterValue _weightmapeffectsmid "Velocity" (.VelocityStrength_14_FC9BD95E42791F6B55C07BAAF882F358 (Utilities|Struct|BreakErodeWeightmap _array_element)) :!class /Script/Engine.MaterialInstanceDynamic)
          (Rendering|Material|SetTextureParameterValue _weightmapeffectsmid "VelocityRT" (select (Utilities|Enum|Equal(Enum) (Variables|/Landmass/Landscape/OldPrototype_BP/BP_Landmass_LayerStack.BP_Landmass_LayerStack_C|GetErosion%20Mode (Variables|Default|GetOwningLayerStackBP)) "NewEnumerator0") (Variables|/Landmass/Landscape/OldPrototype_BP/BP_Landmass_LayerStack.BP_Landmass_LayerStack_C|GetMesh%20Depth%20RT (Variables|Default|GetOwningLayerStackBP)) (Utilities|VelocityPingPong)))
          (Rendering|Material|SetTextureParameterValue _weightmapeffectsmid "HeightRT" _source_rt)
          (Rendering|Material|SetTextureParameterValue _weightmapeffectsmid "HeightRTPrev" _target_rt)
          (Rendering|Material|SetScalarParameterValue _weightmapeffectsmid "ScatterParticles" (Math|Conversions|ToFloat(Boolean) (Utilities|Enum|Equal(Enum) (Variables|/Landmass/Landscape/OldPrototype_BP/BP_Landmass_LayerStack.BP_Landmass_LayerStack_C|GetErosion%20Mode (Variables|Default|GetOwningLayerStackBP)) "NewEnumerator0")) :!class /Script/Engine.MaterialInstanceDynamic)
          (Rendering|DrawMaterialtoRenderTarget 0 (Utilities|Struct|BreakRenderTargetChannels (Utilities|Map|Find _weightmaps (Utilities|String|StringToName (Utilities|Array|Get(aref) (Utilities|String|ParseIntoArray (Utilities|Struct|BreakErodeWeightmap _array_element) "."))))) (Variables|MIDs|GetWeightmapEffectsMID)))))))
